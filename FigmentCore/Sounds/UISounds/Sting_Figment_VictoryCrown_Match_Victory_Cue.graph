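# Sound cue for the Figment Victory Crown match victory sting, a UI sound built on the VictoryCrown VictorySplash wave.
# A very short fixed Delay node feeds a single Wave Player, with the cue's volume boosted.
# It uses the Victory Royale UI sound class, allows one instance through a PreventNew concurrency override, and lists a submix send to the music audio-analysis submix.
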SoundCue /FigmentCore/Sounds/UISounds/Sting_Figment_VictoryCrown_Match_Victory_Cue
root=Delay_0
nodes [
 {id=Delay_0, children [WavePlayer_1], properties {delayMax=0.01, delayMin=0.01}}
 {id=WavePlayer_1, properties {soundWaveAssetPtr=/VictoryCrownsGameplay/Sounds/wav/Sting_VictoryCrown_VictorySplash}}
]
properties {
 maxDistance=2097152
 soundClassObject=/Game/Sounds/FortSoundClassesAndModes/Fort_UI_VictoryRoyale_PSM
 volumeMultiplier=2.1
 attenuationOverrides {falloffDistance=7000, attenuationShapeExtents {x=500}}
 concurrencyOverrides {maxCount=1, resolutionRule=PreventNew}
 soundSubmixSends [{minSendLevel=0, soundSubmix=&Fortnite_AudioAnalysis_Music}]
}
